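# Translucent UI font material for resource text in the player hit point HUD, styled by IsMax, IsEmpty and IsSelected parameters.
# Fill color blends white to cyan at max, then to adaptive Dim when empty and adaptive Dark when selected, with matching opacity blends.
# An IsOutline static switch instead gives Dark color and a vertical UV opacity gradient that vanishes on selection.
# A low-end switcher function supplies empty-state values.
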Material /BattleRoyaleGameUI/HUD/PlayerHitPointInfo/Assets/M_Font_Resources
graph {
 properties {bAutomaticallySetUsageInEditor=false, blendMode=BLEND_Translucent, materialDomain=MD_UI, usedShadingModels=Default Lit}
 nodes [
  MaterialFunctionCall() {materialFunction=&GetUserInterfaceUV}
  n1 = MaterialFunctionCall!(BufferSource) {materialFunction=/BattleRoyaleGameUI/Shared/MF_UI_AdaptiveColors}
  n2 = LinearInterpolate(
   A=ComponentMask(None=CollectionParameter() {collection=/BattleRoyaleGameUI/Shared/MPC_BRGameUI, parameterName=White}) {b=true, g=true, r=true}
   B=ComponentMask(None=CollectionParameter() {collection=&MPC_BRGameUI, parameterName=Cyan}) {b=true, g=true, r=true}
   Alpha=ScalarParameter() {parameterName=IsMax}
  )
  n3 = MaterialFunctionCall!(Default=@n1.Dim, LowEnd=@n1.Dim) {materialFunction=/BattleRoyaleGameUI/Shared/MF_UI_LowEnd_ValueSwitcher}
  n4 = ScalarParameter() {parameterName=IsEmpty}
  StaticSwitchParameter(
   True=@n1.Dark
   False=LinearInterpolate(A=LinearInterpolate(A=@n2, B=@n3.Result, Alpha=@n4), B=@n1.Dark, Alpha=ScalarParameter() {parameterName=IsSelected})
  ) {parameterName=IsOutline}
  n5 = ScalarParameter() {defaultValue=0.5, parameterName=Unselected_OpacityStart}
  n6 = ScalarParameter() {defaultValue=0.25, parameterName=Unselected_OpacityEnd}
  n7 = ComponentMask(None=TextureCoordinate()) {g=true}
  n8 = MaterialFunctionCall(Default=Constant() {r=0.75}, LowEnd=Constant() {r=0.464}) {materialFunction=&MF_UI_LowEnd_ValueSwitcher}
  n9 = ScalarParameter() {parameterName=IsEmpty}
  StaticSwitchParameter(
   True=LinearInterpolate(A=LinearInterpolate(A=@n5, B=@n6, Alpha=@n7) {constB=0}, Alpha=ScalarParameter() {parameterName=IsSelected}) {constB=0}
   False=LinearInterpolate(A=LinearInterpolate(B=@n8.Result, Alpha=@n9) {constA=0.95}, B=Constant() {r=0.9}, Alpha=ScalarParameter() {parameterName=IsSelected})
  ) {parameterName=IsOutline}
 ]
}
