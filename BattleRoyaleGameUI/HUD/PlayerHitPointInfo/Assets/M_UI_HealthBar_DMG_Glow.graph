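# UI material that draws the damage glow on the player's health bar, tinted for health, shield or vehicle damage.
# A rounded SDF box sized by DamageProgress, softened and eased, is masked by a horizontal gradient driven by DamageCurrent and DamageProgress.
# DamageAmount scales intensity, which also blends falloff colors toward core colors, with health and shield cores from the shared UI parameter collection.
# Alpha-composite blending with AdditiveRatio tunes how additive the glow looks, and Anim_Opacity fades it.
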Material /BattleRoyaleGameUI/HUD/PlayerHitPointInfo/Assets/M_UI_HealthBar_DMG_Glow
graph {
 properties {bAutomaticallySetUsageInEditor=false, blendMode=BLEND_AlphaComposite, materialDomain=MD_UI, usedShadingModels=Default Lit}
 nodes [
  n1 = VectorParameter() {parameterName=Color_Falloff_Health, defaultValue=<rgba 0.05 0 0.0167 1>}
  n2 = VectorParameter() {parameterName=Color_Falloff_Shield, defaultValue=<rgba 0 0 0.125 1>}
  n3 = ScalarParameter() {parameterName=IsShieldDMG}
  n4 = LinearInterpolate(
   A=CollectionParameter() {collection=/BattleRoyaleGameUI/Shared/MPC_BRGameUI, parameterName=Red}
   B=CollectionParameter() {collection=&MPC_BRGameUI, parameterName=Cyan}
   Alpha=ScalarParameter() {parameterName=IsShieldDMG}
  )
  n5 = LinearInterpolate(Alpha=ScalarParameter() {defaultValue=1, parameterName=DamageProgress})
  n6 = MaterialFunctionCall() {materialFunction=&GetUserInterfaceUV}
  n7 = SmoothStep(
   Min=LinearInterpolate(Alpha=ScalarParameter() {defaultValue=0.5, parameterName=DamageCurrent}) {constB=0.5}
   Max=@n5
   Value=ComponentMask(None=@n6.Normalized UV) {r=true}
  )
  n8 = Saturate(None=MaterialFunctionCall("Value (0-1)"=Saturate(None=@n7)) {materialFunction=&MF_UI_EaseCurves}.EaseInSine)
  n9 = Multiply(A=OneMinus(None=@n5)) {constB=0.5}
  n10 = Constant()
  n11 = MaterialFunctionCall(
   UVs=Add(A=@n6.Normalized UV, B=AppendVector(A=@n9, B=@n10))
   Size=AppendVector(A=@n5, B=Constant() {r=1})
   CornerRoundness=ScalarParameter() {defaultValue=0.8, parameterName=GlowRoundness}
   KeepAspectRatio=StaticBool() {value=true}
  ) {materialFunction=&MF_UI_Box}
  n12 = MaterialFunctionCall("Value (0-1)"=MaterialFunctionCall(DistanceField=@n11.SDF, Threshold=Constant() {r=-0.5}, Smoothing=Constant() {r=1}) {materialFunction=&MF_UI_AntiAliasing}.Inner) {materialFunction=&MF_UI_EaseCurves}
  n13 = Multiply(A=@n8, B=@n12.EaseInSine)
  n14 = Multiply(
   A=@n13
   B=LinearInterpolate(
    B=ScalarParameter() {defaultValue=1, parameterName=DamageIntensityMultiplier}
    Alpha=ScalarParameter() {defaultValue=0.1, parameterName=DamageAmount}
   ) {constA=0.5, constB=6}
  )
  n15 = LinearInterpolate(
   A=LinearInterpolate(A=LinearInterpolate(A=@n1.RGB, B=@n2.RGB, Alpha=@n3), B=ComponentMask(None=@n4) {b=true, g=true, r=true}, Alpha=@n14)
   B=LinearInterpolate(
    A=VectorParameter() {parameterName=Color_Falloff_Vehicle, defaultValue=<rgba 0.25 0.0417 0 1>}.RGB
    B=VectorParameter() {parameterName=Color_Core_Vehicle, defaultValue=<rgba 1 0.85 0.1 1>}.RGB
    Alpha=@n14
   )
   Alpha=ScalarParameter() {parameterName=IsVehicleDMG}
  )
  Multiply(A=VertexColor().A, B=Multiply(A=Multiply(A=@n15, B=@n14), B=ScalarParameter() {defaultValue=1, parameterName=Anim_Opacity}))
  Multiply(
   A=Multiply(A=@n13, B=ScalarParameter() {defaultValue=0.5, parameterName=AdditiveRatio})
   B=ScalarParameter() {defaultValue=1, parameterName=Anim_Opacity}
  )
  Multiply(A=LinearInterpolate(Alpha=ScalarParameter() {defaultValue=0.1, parameterName=DamageAmount}) {constA=0.666, constB=0.5}) {constB=-1}
  Saturate(None=@n12.EaseInSine)
  Saturate(None=Multiply(A=@n8, B=@n11.Fill))
 ]
}
