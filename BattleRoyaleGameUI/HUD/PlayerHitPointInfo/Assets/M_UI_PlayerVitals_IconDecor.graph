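# Translucent UI material that draws a decorative angled shape, named as icon decor for the player vitals HUD in battle royale.
# Rotated linear gradients define the left notch and the left, top and bottom edges, merged with Min and anti-aliasing.
# A smoothstepped box SDF cutout and an eased, rotated opacity fade multiply that mask.
# MF_UI_AdaptiveColors supplies colors, though the vertical gradient blends Bright with itself.
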Material /BattleRoyaleGameUI/HUD/PlayerHitPointInfo/Assets/M_UI_PlayerVitals_IconDecor
graph {
 properties {bAutomaticallySetUsageInEditor=false, blendMode=BLEND_Translucent, materialDomain=MD_UI, usedShadingModels=Default Lit}
 nodes [
  n1 = NamedRerouteUsage()
  Add(A=@n1, B=Constant2Vector() {r=0.25})
  n2 = ScalarParameter() {defaultValue=-0.3, parameterName=BoxCutout_OffsetX}
  n3 = ScalarParameter() {defaultValue=0.0775, parameterName=BoxCutout_OffsetY}
  n4 = MaterialFunctionCall(
   UVs=Add(A=NamedRerouteUsage(), B=AppendVector(A=@n2, B=@n3))
   Size=AppendVector(A=ScalarParameter() {defaultValue=1, parameterName=BoxCutoutSize_X}, B=ScalarParameter() {defaultValue=0.4, parameterName=BoxCutoutSize_Y})
   CornerRoundness=ScalarParameter() {defaultValue=-0.0333, parameterName=ShadowCornerRoundess}
   KeepAspectRatio=StaticBool() {value=true}
  ) {materialFunction=&MF_UI_Box}
  n5 = SmoothStep(Min=ScalarParameter() {parameterName=ShadowMin}, Max=ScalarParameter() {defaultValue=1, parameterName=ShadowMax}, Value=@n4.SDF)
  LinearInterpolate(A=MaterialFunctionCall!(BufferSource) {materialFunction=/BattleRoyaleGameUI/Shared/MF_UI_AdaptiveColors}.Dark, Alpha=@n5)
  LinearInterpolate(B=Saturate(None=OneMinus(None=@n4.Fill)), Alpha=ScalarParameter() {defaultValue=1, parameterName=ShadowAmount}) {constA=1}
  n6 = MaterialFunctionCall() {materialFunction=&MF_UI_AdaptiveColors}
  n7 = ComponentMask(None=MaterialFunctionCall() {materialFunction=&GetUserInterfaceUV}.Normalized UV) {g=true}
  LinearInterpolate(
   A=@n6.Bright
   B=@n6.Bright
   Alpha=Saturate(None=MaterialFunctionCall("Value (0-1)"=@n7) {materialFunction=&MF_UI_EaseCurves}.EaseInSine)
  )
  n8 = MaterialFunctionCall(
   UVs=NamedRerouteUsage()
   RotationAngle=ScalarParameter() {defaultValue=0.165, parameterName=LetCutNotch_Rotation}
   GradientOffset=ScalarParameter() {defaultValue=-0.07, parameterName=LeftCutNotch_Offset}
  ) {materialFunction=&MF_UI_LinearGradient}
  n9 = MaterialFunctionCall() {materialFunction=&GetUserInterfaceUV}
  n10 = OneMinus(None=MaterialFunctionCall(
    UVs=NamedRerouteDeclaration(None=@n9.Normalized UV) {name=NormalizedUVs, nodeColor=<rgba 0 0 0 1>}
    RotationAngle=ScalarParameter() {defaultValue=0.02, parameterName=LeftEdge_1_Rotation}
    GradientOffset=ScalarParameter() {defaultValue=-0.4025, parameterName=LeftEdge_1_Offset}
   ) {materialFunction=&MF_UI_LinearGradient}.Result)
  n11 = NamedRerouteUsage()
  n12 = ScalarParameter() {defaultValue=0.0185, parameterName=LeftEdge_2_Rotation}
  n13 = ScalarParameter() {defaultValue=-0.44, parameterName=LeftEdge_2_Offset}
  n14 = Min(
   A=MaterialFunctionCall(DistanceField=Min(A=@n8.Result, B=@n10), Threshold=Constant() {r=0.5}) {materialFunction=&MF_UI_AntiAliasing}.Inner
   B=MaterialFunctionCall(
    DistanceField=MaterialFunctionCall(UVs=@n11, RotationAngle=@n12, GradientOffset=@n13) {materialFunction=&MF_UI_LinearGradient}.Result
    Threshold=Constant() {r=0.5}
   ) {materialFunction=&MF_UI_AntiAliasing}.Outer
  )
  n15 = MaterialFunctionCall(
   UVs=NamedRerouteUsage()
   RotationAngle=ScalarParameter() {defaultValue=0.2125, parameterName=TopEdge_Rotation}
   GradientOffset=ScalarParameter() {defaultValue=0.365, parameterName=TopEdge_Offset}
  ) {materialFunction=&MF_UI_LinearGradient}
  n16 = MaterialFunctionCall(
   UVs=NamedRerouteUsage()
   RotationAngle=ScalarParameter() {defaultValue=-0.225, parameterName=BottomEdgeRotation}
   GradientOffset=ScalarParameter() {defaultValue=0.265, parameterName=BottomEdge_Offset}
  ) {materialFunction=&MF_UI_LinearGradient}
  n17 = MaterialFunctionCall(DistanceField=Min(A=OneMinus(None=@n15.Result), B=OneMinus(None=@n16.Result)), Threshold=Constant() {r=0.5}) {materialFunction=&MF_UI_AntiAliasing}
  n18 = MaterialFunctionCall(UVs=@n1, "Rotation Angle (0-1)"=ScalarParameter() {defaultValue=0.275, parameterName=FadeRotation}) {materialFunction=&CustomRotator}
  n19 = Saturate(None=Saturate(None=ComponentMask(None=@n18.Rotated Values) {r=true}))
  Multiply(
   A=Multiply(A=Min(A=@n14, B=@n17.Outer), B=@n5)
   B=LinearInterpolate(
    A=ScalarParameter() {defaultValue=0.75, parameterName=Opacity_Start}
    B=ScalarParameter() {defaultValue=0.25, parameterName=Opacity_Stop}
    Alpha=MaterialFunctionCall("Value (0-1)"=@n19) {materialFunction=&MF_UI_EaseCurves}.EaseInSine
   )
  )
 ]
}
