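# Translucent UI master material for player vitals icons in the Battle Royale HUD hit point display.
# A selectable channel of the hit point icon texture is thresholded into icon and optional outline masks, with customized UVs for rotation and scale.
# Progress at or below zero switches the icon to empty opacity and dimmed adaptive color.
# Anim_Color blends in red, cyan shield or custom vehicle damage tints, and a low-end switcher adjusts empty alpha.
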Material /BattleRoyaleGameUI/HUD/PlayerHitPointInfo/Assets/M_UI_PlayerVitals_Icon_Master
graph {
 properties {bAutomaticallySetUsageInEditor=false, blendMode=BLEND_Translucent, materialDomain=MD_UI, numCustomizedUVs=1, usedShadingModels=Default Lit}
 nodes [
  n1 = MaterialFunctionCall(UVs=TextureCoordinate(), Texture Scale=ScalarParameter() {defaultValue=0.85, parameterName=IconScale}) {materialFunction=&ScaleUVsByCenter}
  n2 = Constant()
  n3 = Constant()
  n4 = StaticComponentMaskParameter(None=TextureSample(UVs=MaterialFunctionCall!(UV=@n1.UVs, Perspective Rotation=@n2, Perspective Amount=@n3) {materialFunction=/FortUILibrary/MaterialFunctions/Basics/MF_UI_Perspective}.Perspective UVs) {samplerType=SAMPLERTYPE_LinearColor, texture=/Game/Athena/HUD/Materials/T_UI_HitPoint_Type_Icon}.RGB) {defaultG=true, parameterName=Channel}
  n5 = ScalarParameter() {defaultValue=0.25, parameterName=Subtract_Edge}
  n6 = Saturate(None=Multiply(A=Subtract(A=@n4, B=@n5), B=ScalarParameter() {defaultValue=8, parameterName=Multiply_Edge}))
  n7 = LinearInterpolate(
   A=ScalarParameter() {defaultValue=0.95, parameterName=Alpha}
   B=MaterialFunctionCall!(Default=ScalarParameter() {defaultValue=0.75, parameterName=EmptyAlpha}, LowEnd=Constant() {r=0.464}) {materialFunction=/BattleRoyaleGameUI/Shared/MF_UI_LowEnd_ValueSwitcher}.Result
   Alpha=NamedRerouteUsage()
  ) {constA=1}
  n8 = ScalarParameter() {parameterName=Outline_Subtract_Edge}
  n9 = Saturate(None=Multiply(A=Subtract(A=@n4, B=@n8), B=ScalarParameter() {defaultValue=1, parameterName=Outline_Multiply_Edge}))
  n10 = ScalarParameter() {defaultValue=0.5, parameterName=Outline_Alpha}
  LinearInterpolate(
   A=Multiply(A=@n6, B=@n7)
   B=LinearInterpolate(A=Multiply(A=@n9, B=@n10), B=@n7, Alpha=@n6)
   Alpha=ScalarParameter() {defaultValue=1, parameterName=UseOutline}
  )
  n11 = MaterialFunctionCall!(BufferSource) {materialFunction=/BattleRoyaleGameUI/Shared/MF_UI_AdaptiveColors}
  n12 = MaterialFunctionCall(Default=@n11.Dim, LowEnd=@n11.Dim) {materialFunction=&MF_UI_LowEnd_ValueSwitcher}
  n13 = NamedRerouteUsage()
  n14 = ComponentMask(None=LinearInterpolate(
    A=CollectionParameter() {collection=/BattleRoyaleGameUI/Shared/MPC_BRGameUI, parameterName=Red}
    B=CollectionParameter() {collection=&MPC_BRGameUI, parameterName=Cyan}
    Alpha=ScalarParameter() {parameterName=IsShieldDMG}
   )) {b=true, g=true, r=true}
  n15 = VectorParameter() {parameterName=Color, defaultValue=<rgba 1 0.85 0.1 1>}
  n16 = ScalarParameter() {parameterName=IsVehicleDMG}
  LinearInterpolate(
   A=@n11.Dark
   B=LinearInterpolate(
    A=LinearInterpolate(A=@n11.White, B=@n12.Result, Alpha=@n13)
    B=LinearInterpolate(A=@n14, B=@n15.RGB, Alpha=@n16)
    Alpha=ScalarParameter() {parameterName=Anim_Color}
   )
   Alpha=@n6
  )
  MaterialFunctionCall(
   UVs=MaterialFunctionCall(UVs=TextureCoordinate(), "Rotation Angle (0-1)"=ScalarParameter() {parameterName=Rotation}) {materialFunction=&CustomRotator}.Rotated Values
   Texture Scale=ScalarParameter() {defaultValue=1, parameterName=Scale}
  ) {materialFunction=&ScaleUVsByCenter}
  NamedRerouteDeclaration(None=MaterialFunctionCall!(A=ScalarParameter() {defaultValue=0.6, group=Fill, parameterName=Progress, sortPriority=0}, B=Constant()) {materialFunction=/FortUILibrary/MaterialFunctions/Basics/MF_UI_BooleanOperations}.IsLessOrEqual) {name=EmptyCalculation, nodeColor=<rgba 0 0 0 1>}
 ]
}
