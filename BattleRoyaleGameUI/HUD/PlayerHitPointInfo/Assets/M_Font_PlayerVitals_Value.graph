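# Translucent UI font material that colors the numeric value text of the player vitals HUD, such as health and shield.
# A Progress check at zero switches white text to the adaptive Dim color and lowers alpha for the empty state.
# Anim_Damage blends in a damage tint: red, cyan for shield, or VehicleColor for vehicles.
# A low-end switcher supplies a fixed empty alpha in place of the EmptyAlpha parameter.
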Material /BattleRoyaleGameUI/HUD/PlayerHitPointInfo/Assets/M_Font_PlayerVitals_Value
graph {
 properties {bAutomaticallySetUsageInEditor=false, blendMode=BLEND_Translucent, materialDomain=MD_UI, usedShadingModels=Default Lit}
 nodes [
  n1 = MaterialFunctionCall!(A=ScalarParameter() {defaultValue=0.5, group=Fill, parameterName=Progress, sortPriority=0}, B=Constant()) {materialFunction=/FortUILibrary/MaterialFunctions/Basics/MF_UI_BooleanOperations}
  LinearInterpolate(
   B=MaterialFunctionCall!(Default=ScalarParameter() {defaultValue=0.75, parameterName=EmptyAlpha}, LowEnd=Constant() {r=0.464}) {materialFunction=/BattleRoyaleGameUI/Shared/MF_UI_LowEnd_ValueSwitcher}.Result
   Alpha=@n1.IsLessOrEqual
  ) {constA=0.95}
  n2 = CollectionParameter() {collection=/BattleRoyaleGameUI/Shared/MPC_BRGameUI, parameterName=White}
  n3 = MaterialFunctionCall!(BufferSource) {materialFunction=/BattleRoyaleGameUI/Shared/MF_UI_AdaptiveColors}
  n4 = ComponentMask(None=LinearInterpolate(
    A=CollectionParameter() {collection=&MPC_BRGameUI, parameterName=Red}
    B=CollectionParameter() {collection=&MPC_BRGameUI, parameterName=Cyan}
    Alpha=ScalarParameter() {parameterName=IsShieldDMG}
   )) {b=true, g=true, r=true}
  n5 = VectorParameter() {parameterName=VehicleColor, defaultValue=<rgba 1 0.85 0.1 1>}
  n6 = ScalarParameter() {parameterName=IsVehicleDMG}
  LinearInterpolate(
   A=LinearInterpolate(
    A=ComponentMask(None=@n2) {b=true, g=true, r=true}
    B=MaterialFunctionCall(Default=@n3.Dim, LowEnd=@n3.Dim) {materialFunction=&MF_UI_LowEnd_ValueSwitcher}.Result
    Alpha=@n1.IsLessOrEqual
   )
   B=Saturate(None=LinearInterpolate(A=@n4, B=@n5.RGB, Alpha=@n6))
   Alpha=ScalarParameter() {parameterName=Anim_Damage}
  )
 ]
}
